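# Sound cue for the gold Brute monster's attack vocalizations, under the Deimos monster sounds.
# An equally weighted Random node picks one of five Brute attack waves and feeds it through a Modulator.
# The cue lowers pitch substantially and boosts volume, with high priority and virtualization disabled.
# It routes to a creature vocalization sound class, using close-range Deimos vocal attenuation and both global and local Deimos vocal concurrency sets.
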SoundCue /Game/Sounds/Fort_Monster_Deimos/Brute/BruteGold_VO_Attack_Cue
root=Modulator_0
nodes [
 {id=Modulator_0, children [Random_1]}
 {
  id=Random_1
  children [WavePlayer_7, WavePlayer_8, WavePlayer_9, WavePlayer_10, WavePlayer_11]
  properties {bShouldExcludeFromBranchCulling=true, weights [1,1,1,1,1]}
 }
 {id=WavePlayer_10, properties {soundWaveAssetPtr=&/Brute_Attack_04}}
 {id=WavePlayer_11, properties {soundWaveAssetPtr=&/Brute_Attack_05}}
 {id=WavePlayer_7, properties {soundWaveAssetPtr=&/Brute_Attack_01}}
 {id=WavePlayer_8, properties {soundWaveAssetPtr=&/Brute_Attack_02}}
 {id=WavePlayer_9, properties {soundWaveAssetPtr=&/Brute_Attack_03}}
]
properties {
 attenuationSettings=/Game/Sounds/Attenuation/Fort_Deimos_Vocal_Close_Att
 bBypassVolumeScaleForPriority=true
 maxDistance=13000
 pitchMultiplier=0.6
 priority=0.9
 soundClassObject=&Fort_Creatures_Vocalization_Tactical
 virtualizationMode=Disabled
 volumeMultiplier=2.9925
 attenuationOverrides {distanceAlgorithm=NaturalSound, falloffDistance=4200, attenuationShapeExtents {x=800}}
 concurrencyOverrides {resolutionRule=PreventNew}
 concurrencySet [/Game/Sounds/Fort_Monster_Deimos/Settings/Deimos_Vocal_Global_Conc, /Game/Sounds/Fort_Monster_Deimos/Settings/Deimos_Vocal_Local_Conc]
}
