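# Sound cue for the gold Brute monster's death vocalization, mixing a close layer with a distant layer.
# Each layer picks randomly from three death waves through its own attenuation asset, with the distant layer also passing through a Modulator and mixed slightly quieter.
# The whole cue is pitched well down and boosted in volume, with virtualization disabled and shared Deimos global and local vocal concurrency sets.
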SoundCue /Game/Sounds/Fort_Monster_Deimos/Brute/BruteGold_VO_Death_Cue
root=Mixer_0
nodes [
 {id=Attenuation_0, children [Random_1], properties {attenuationSettings=/Game/Sounds/Attenuation/Fort_Deimos_Vocal_Close_Att}}
 {id=Attenuation_1, children [Modulator_1], properties {attenuationSettings=/Game/Sounds/Attenuation/Fort_Deimos_Vocal_Distant_Att}}
 {id=Mixer_0, children [Attenuation_0, Attenuation_1], properties {inputVolume [1,0.8]}}
 {id=Modulator_1, children [Random_2]}
 {id=Random_1, children [WavePlayer_4,WavePlayer_5,WavePlayer_6], properties {bShouldExcludeFromBranchCulling=true, weights [1,1,1]}}
 {id=Random_2, children [WavePlayer_8, WavePlayer_9, WavePlayer_10], properties {bShouldExcludeFromBranchCulling=true, weights [1,1,1]}}
 {id=WavePlayer_10, properties {soundWaveAssetPtr=&/Brute_Death_Distant_03}}
 {id=WavePlayer_4, properties {soundWaveAssetPtr=&/Brute_Death_01}}
 {id=WavePlayer_5, properties {soundWaveAssetPtr=&/Brute_Death_02}}
 {id=WavePlayer_6, properties {soundWaveAssetPtr=&/Brute_Death_03}}
 {id=WavePlayer_8, properties {soundWaveAssetPtr=&/Brute_Death_Distant_01}}
 {id=WavePlayer_9, properties {soundWaveAssetPtr=&/Brute_Death_Distant_02}}
]
properties {
 bBypassVolumeScaleForPriority=true
 maxDistance=24000
 pitchMultiplier=0.6
 priority=0.75
 soundClassObject=&Fort_Creatures_Vocalization_Tactical
 virtualizationMode=Disabled
 volumeMultiplier=3.0528
 concurrencyOverrides {resolutionRule=PreventNew}
 concurrencySet [/Game/Sounds/Fort_Monster_Deimos/Settings/Deimos_Vocal_Global_Conc, /Game/Sounds/Fort_Monster_Deimos/Settings/Deimos_Vocal_Local_Conc]
}
