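# Sound cue for the pain vocalization of the gold Deimos Brute monster, pitched down across the whole cue.
# A mixer layers a close branch, briefly delayed and louder, with a quieter distant branch, each under its own vocal attenuation.
# Each branch modulates a random pick from five Brute pain waves, close or distant recordings respectively.
# Global and local Deimos vocal concurrency sets limit playback, and virtualization is disabled.
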SoundCue /Game/Sounds/Fort_Monster_Deimos/Brute/BruteGold_VO_Pain_Cue
root=Mixer_0
nodes [
 {id=Attenuation_0, children [Delay_0], properties {attenuationSettings=/Game/Sounds/Attenuation/Fort_Deimos_Vocal_Close_Att}}
 {id=Attenuation_1, children [Modulator_1], properties {attenuationSettings=/Game/Sounds/Attenuation/Fort_Deimos_Vocal_Distant_Att}}
 {id=Delay_0, children [Modulator_0], properties {delayMax=0.1, delayMin=0.1}}
 {id=Mixer_0, children [Attenuation_0, Attenuation_1], properties {inputVolume [1.25,0.7]}}
 {id=Modulator_0, children [Random_1]}
 {id=Modulator_1, children [Random_2]}
 {
  id=Random_1
  children [WavePlayer_5,WavePlayer_6,WavePlayer_7,WavePlayer_8,WavePlayer_9]
  properties {bShouldExcludeFromBranchCulling=true, weights [1,1,1,1,1]}
 }
 {
  id=Random_2
  children [WavePlayer_10, WavePlayer_12, WavePlayer_13, WavePlayer_14, WavePlayer_15]
  properties {bShouldExcludeFromBranchCulling=true, weights [1,1,1,1,1]}
 }
 {id=WavePlayer_10, properties {soundWaveAssetPtr=&/Brute_Pain_Distant_01}}
 {id=WavePlayer_12, properties {soundWaveAssetPtr=&/Brute_Pain_Distant_02}}
 {id=WavePlayer_13, properties {soundWaveAssetPtr=&/Brute_Pain_Distant_03}}
 {id=WavePlayer_14, properties {soundWaveAssetPtr=&/Brute_Pain_Distant_04}}
 {id=WavePlayer_15, properties {soundWaveAssetPtr=&/Brute_Pain_Distant_05}}
 {id=WavePlayer_5, properties {soundWaveAssetPtr=&/Brute_Pain_01}}
 {id=WavePlayer_6, properties {soundWaveAssetPtr=&/Brute_Pain_02}}
 {id=WavePlayer_7, properties {soundWaveAssetPtr=&/Brute_Pain_03}}
 {id=WavePlayer_8, properties {soundWaveAssetPtr=&/Brute_Pain_04}}
 {id=WavePlayer_9, properties {soundWaveAssetPtr=&/Brute_Pain_05}}
]
properties {
 bBypassVolumeScaleForPriority=true
 maxDistance=24000
 pitchMultiplier=0.6
 priority=0.75
 soundClassObject=&Fort_Creatures_Vocalization_Tactical
 virtualizationMode=Disabled
 volumeMultiplier=1.7955
 attenuationOverrides {distanceAlgorithm=NaturalSound, falloffDistance=4000, attenuationShapeExtents {x=1000}}
 concurrencyOverrides {resolutionRule=PreventNew}
 concurrencySet [/Game/Sounds/Fort_Monster_Deimos/Settings/Deimos_Vocal_Global_Conc, /Game/Sounds/Fort_Monster_Deimos/Settings/Deimos_Vocal_Local_Conc]
}
