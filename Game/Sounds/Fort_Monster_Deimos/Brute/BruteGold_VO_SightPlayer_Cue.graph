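# Sound cue for the Gold Brute monster's vocalization when it sights a player.
# After a short random delay, a Mixer plays two layers at once: a random Brute attack clip with close attenuation, and a random sight-player clip with distant attenuation.
# Each layer passes through a Modulator for variation, while the whole cue is pitched down and boosted in volume.
# Global and local Deimos vocal concurrency sets are assigned to the cue.
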SoundCue /Game/Sounds/Fort_Monster_Deimos/Brute/BruteGold_VO_SightPlayer_Cue
root=Delay_0
nodes [
 {id=Attenuation_0, children [Modulator_0], properties {attenuationSettings=/Game/Sounds/Attenuation/Fort_Deimos_Vocal_Close_Att}}
 {
  id=Attenuation_1
  children [Modulator_1]
  properties {
   attenuationSettings=/Game/Sounds/Attenuation/Fort_Deimos_Vocal_Distant_Att
   attenuationOverrides {falloffDistance=7000, stereoSpread=6000, attenuationShapeExtents {x=5000}}
  }
 }
 {id=Delay_0, children [Mixer_0], properties {delayMax=0.75, delayMin=0.25}}
 {id=Mixer_0, children [Attenuation_0, Attenuation_1], properties {inputVolume [1,1]}}
 {id=Modulator_0, children [Random_1]}
 {id=Modulator_1, children [Random_0], properties {pitchMax=1.2}}
 {
  id=Random_0
  children [WavePlayer_0,WavePlayer_1,WavePlayer_2,WavePlayer_3,WavePlayer_4]
  properties {bShouldExcludeFromBranchCulling=true, weights [1,1,1,1,1]}
 }
 {
  id=Random_1
  children [WavePlayer_7, WavePlayer_8, WavePlayer_9, WavePlayer_10, WavePlayer_11]
  properties {bShouldExcludeFromBranchCulling=true, weights [1,1,1,1,1]}
 }
 {id=WavePlayer_0, properties {soundWaveAssetPtr=&/Brute_SightPlayer_01}}
 {id=WavePlayer_1, properties {soundWaveAssetPtr=&/Brute_SightPlayer_02}}
 {id=WavePlayer_10, properties {soundWaveAssetPtr=&/Brute_Attack_04}}
 {id=WavePlayer_11, properties {soundWaveAssetPtr=&/Brute_Attack_05}}
 {id=WavePlayer_2, properties {soundWaveAssetPtr=&/Brute_SightPlayer_03}}
 {id=WavePlayer_3, properties {soundWaveAssetPtr=&/Brute_SightPlayer_04}}
 {id=WavePlayer_4, properties {soundWaveAssetPtr=&/Brute_SightPlayer_05}}
 {id=WavePlayer_7, properties {soundWaveAssetPtr=&/Brute_Attack_01}}
 {id=WavePlayer_8, properties {soundWaveAssetPtr=&/Brute_Attack_02}}
 {id=WavePlayer_9, properties {soundWaveAssetPtr=&/Brute_Attack_03}}
]
properties {
 bBypassVolumeScaleForPriority=true
 maxDistance=24000
 pitchMultiplier=0.6
 priority=0.75
 soundClassObject=&Fort_Creatures_Vocalization_Tactical
 volumeMultiplier=2.9925
 attenuationOverrides {distanceAlgorithm=NaturalSound, falloffDistance=4200, attenuationShapeExtents {x=800}}
 concurrencyOverrides {maxCount=3, resolutionRule=StopFarthestThenPreventNew, volumeScale=0.75, volumeScaleAttackTime=0.1}
 concurrencySet [/Game/Sounds/Fort_Monster_Deimos/Settings/Deimos_Vocal_Global_Conc, /Game/Sounds/Fort_Monster_Deimos/Settings/Deimos_Vocal_Local_Conc]
}
