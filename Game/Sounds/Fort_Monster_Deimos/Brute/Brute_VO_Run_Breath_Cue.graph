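# Sound cue for the Deimos Brute monster's breathing vocalization while running.
# A Random node picks one of ten equally weighted breath waves, then feeds a Modulator at the root for variation.
# It routes to the creature vocalization sound class with close vocal attenuation, boosted volume and shared global and local Deimos vocal concurrency sets.
# Virtualization is disabled, and priority ignores volume scaling.
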SoundCue /Game/Sounds/Fort_Monster_Deimos/Brute/Brute_VO_Run_Breath_Cue
root=Modulator_1
nodes [
 {id=Modulator_1, children [Random_1]}
 {
  id=Random_1
  children [
   WavePlayer_9
   WavePlayer_10
   WavePlayer_11
   WavePlayer_12
   WavePlayer_13
   WavePlayer_5
   WavePlayer_6
   WavePlayer_7
   WavePlayer_8
   WavePlayer_18
  ]
  properties {bShouldExcludeFromBranchCulling=true, weights [1,1,1,1,1,1,1,1,1,1]}
 }
 {id=WavePlayer_10, properties {soundWaveAssetPtr=&/Brute_RunBreath_01}}
 {id=WavePlayer_11, properties {soundWaveAssetPtr=&/Brute_RunBreath_02}}
 {id=WavePlayer_12, properties {soundWaveAssetPtr=&/Brute_RunBreath_03}}
 {id=WavePlayer_13, properties {soundWaveAssetPtr=&/Brute_RunBreath_04}}
 {id=WavePlayer_18, properties {soundWaveAssetPtr=&/Brute_RunBreath_09}}
 {id=WavePlayer_5, properties {soundWaveAssetPtr=&/Brute_RunBreath_05}}
 {id=WavePlayer_6, properties {soundWaveAssetPtr=&/Brute_RunBreath_06}}
 {id=WavePlayer_7, properties {soundWaveAssetPtr=&/Brute_RunBreath_07}}
 {id=WavePlayer_8, properties {soundWaveAssetPtr=&/Brute_RunBreath_08}}
 {id=WavePlayer_9, properties {soundWaveAssetPtr=&/Brute_RunBreath_10}}
]
properties {
 attenuationSettings=/Game/Sounds/Attenuation/Fort_Deimos_Vocal_Close_Att
 bBypassVolumeScaleForPriority=true
 maxDistance=13000
 priority=0.75
 soundClassObject=/Game/Sounds/FortSoundClassesAndModes/Fort_Creatures_Vocalization_Normal
 virtualizationMode=Disabled
 volumeMultiplier=2.394
 attenuationOverrides {distanceAlgorithm=NaturalSound, falloffDistance=4000, attenuationShapeExtents {x=1000}}
 concurrencyOverrides {maxCount=2, volumeScale=0.75, volumeScaleAttackTime=0.1}
 concurrencySet [/Game/Sounds/Fort_Monster_Deimos/Settings/Deimos_Vocal_Global_Conc, /Game/Sounds/Fort_Monster_Deimos/Settings/Deimos_Vocal_Local_Conc]
}
